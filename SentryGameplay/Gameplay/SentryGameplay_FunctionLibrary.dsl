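; Function library for sentry targeting in Fortnite: target validity, aim point and line of sight.
; Validity checks filter and range, then pawn death and downed rules, or building damageability, required components and sentry tag.
; Aim points use tag-keyed scalable offsets, scaled by capsule half-height for players.
; Sight applies tag and indicated-actor overrides, then line-traces ignoring the target's vehicle, optionally seeing through its melee block or carry partner.
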(fn SentryIsActorValidTarget
  (params (RequestingActor :type Actor%20Object%20Reference) (ActorToCheck :type Actor%20Object%20Reference) (TargetFilter :type Fort%20Target%20Filter%20Structure) (MaxRange :type Float%20%28double-precision%29) (CanTargetDBNO :type Boolean) (CanTargetDBNOCarried :type Boolean) (BuildingComponentsRequiredAnyForTargeting :type Array%20of%20Actor%20Component%20Class%20References) (CanTargetSentryActors :type Boolean))
  (Utilities|IsValid (Variables|GetRequestingActor)
    (:Is%20Valid
      (Utilities|IsValid (Variables|GetActorToCheck)
        (:Is%20Valid
          (bind _returnvalue (Ability|FilterSingleTargetActor (Variables|GetRequestingActor) (Variables|GetTargetFilter) (Variables|GetRequestingActor) (Variables|GetActorToCheck)))
          (if _returnvalue
            (if (<= (Transformation|GetDistanceTo (Variables|GetRequestingActor) (Variables|GetActorToCheck)) (Variables|GetMaxRange))
              (bind _asfort_pawn (Utilities|Casting|CastToFortPawn (Variables|GetActorToCheck))
                (:then
                  (if (not (FortPawn|IsDead _asfort_pawn))
                    (if (Health|IsDBNO _asfort_pawn)
                      (if (GameplayTags|HasMatchingGameplayTag _asfort_pawn "(TagName=\"Gameplay.Status.DBNOBeingCarried\")")
                        (return (Variables|GetCanTargetDBNOCarried))
                        (else
                          (return (Variables|GetCanTargetDBNO))))
                      (else
                        (return true)))
                    (else
                      (return false))))
                (:CastFailed
                  (bind _asbuilding_actor (Utilities|Casting|CastToBuildingActor (Variables|GetActorToCheck))
                    (:then
                      (if (Math|Boolean|ANDBoolean (not (Variables|/Script/FortniteGame.BuildingActor|GetbIsInvulnerable _asbuilding_actor)) (Variables|/Script/Engine.Actor|GetbCanBeDamaged _asbuilding_actor) (Health|HasHealthLeft _asbuilding_actor))
                        (if (Utilities|Array|IsNotEmpty (Variables|GetBuildingComponentsRequiredAnyForTargeting))
                          (LocalVariables|SetHasRequiredBuildingComponent false)
                          (bind _array_element (Utilities|Array|ForEachLoopwithBreak (Variables|GetBuildingComponentsRequiredAnyForTargeting))
                            (:LoopBody
                              (if (Utilities|IsValid (Actor|GetComponentbyClass (Variables|GetActorToCheck) _array_element) :!class /Script/Engine.KismetSystemLibrary)
                                (LocalVariables|SetHasRequiredBuildingComponent true)))
                            (:Completed
                              _))
                          (else
                            (LocalVariables|SetHasRequiredBuildingComponent true)))
                        (if (LocalVariables|GetHasRequiredBuildingComponent)
                          (if (Variables|GetCanTargetSentryActors)
                            (return true)
                            (else
                              (bind _returnvalue_2 (GameplayTags|HasMatchingGameplayTag(Message) (Variables|GetActorToCheck) "(TagName=\"Asset.Athena.EnvItem.Sentry\")"))
                              (return (not _returnvalue_2))))
                          (else
                            (return false)))
                        (else
                          (return false))))
                    (:CastFailed
                      (return false)))))
              (else
                (return false)))
            (else
              (return false))))
        (:Is%20Not%20Valid
          (return false))))
    (:Is%20Not%20Valid
      (return false))))

(fn Sentry_GetAimLocationForTargetActor
  (params (TargetActor :type Actor%20Object%20Reference) (Player_PctCapsuleHalfHeightOffset_Default :type Scalable%20Float%20Structure) (PlayerPawnTargetTagToPctCapsuleHalfHeightZOffset :type Map%20of%20Gameplay%20Tag%20Structures%20to%20Scalable%20Float%20Structures) (NonPlayerPawnTargetTagToZOffsetMap :type Map%20of%20Gameplay%20Tag%20Structures%20to%20Scalable%20Float%20Structures))
  (bind _asfort_player_pawn (Utilities|Casting|CastToFortPlayerPawn (Variables|GetTargetActor))
    (:then
      (LocalVariables|SetPlayerPawn _asfort_player_pawn)
      (bind _keys (Utilities|Map|Keys (Variables|GetPlayerPawnTargetTagToPctCapsuleHalfHeightZOffset)))
      (for _array_element _keys
        (bind _returnvalue (GameplayTags|HasMatchingGameplayTag(Message) (Variables|GetTargetActor) _array_element))
        (if _returnvalue
          (return (Utilities|Operators|Add (Transformation|GetActorLocation (LocalVariables|GetPlayerPawn)) (Math|Vector|MakeVector 0.0 0.0 (* (Components|Capsule|GetScaledCapsuleHalfHeight (Variables|/Script/Engine.Character|GetCapsuleComponent (LocalVariables|GetPlayerPawn))) (ScalableFloat|GetValueatLevel (Utilities|Map|Find (Variables|GetPlayerPawnTargetTagToPctCapsuleHalfHeightZOffset) _array_element))))))))
      (return (Utilities|Operators|Add (Transformation|GetActorLocation (LocalVariables|GetPlayerPawn)) (Math|Vector|MakeVector 0.0 0.0 (* (Components|Capsule|GetScaledCapsuleHalfHeight (Variables|/Script/Engine.Character|GetCapsuleComponent (LocalVariables|GetPlayerPawn))) (ScalableFloat|GetValueatLevel (Variables|GetPlayer_PctCapsuleHalfHeightOffset_Default)))))))
    (:CastFailed
      (bind _keys_2 (Utilities|Map|Keys (Variables|GetNonPlayerPawnTargetTagToZOffsetMap)))
      (for _array_element_2 _keys_2
        (bind _returnvalue_2 (GameplayTags|HasMatchingGameplayTag(Message) (Variables|GetTargetActor) _array_element_2))
        (if _returnvalue_2
          (return (+ (* (Transformation|GetActorUpVector (Variables|GetTargetActor)) (ScalableFloat|GetValueatLevel (Utilities|Map|Find (Variables|GetNonPlayerPawnTargetTagToZOffsetMap) _array_element_2))) (Transformation|GetActorLocation (Variables|GetTargetActor))))))
      (return (Fortnite|GetActorCollisionComponentBounds (Variables|GetTargetActor))))))

(fn Sentry_HasLineOfSightToActor
  (params (TargetActor :type Actor%20Object%20Reference) (Aim_SourceLocation :type Vector) (Aim_TargetLocation :type Vector) (TraceChannel :type ETraceTypeQuery%20Enum :default "TraceTypeQuery1") (Tags_PreventVision :type Gameplay%20Tag%20Container%20Structure) (Tags_AlwaysVisible :type Gameplay%20Tag%20Container%20Structure) (ActorsToIgnore :type Array%20of%20Actor%20Object%20References) (MaxRange :type Float%20%28double-precision%29) (CanTargetDBNOCarried :type Boolean) (CanSeeThroughMeleeBlock :type Boolean) (OwnerController :type Controller%20Object%20Reference) (CanAlwaysSeeIndicatedActorsForOwnerController :type Boolean))
  (bind _returnvalue (GameplayTags|HasAnyMatchingGameplayTags(Message) (Variables|GetTargetActor) (Variables|GetTags_PreventVision)))
  (if (not _returnvalue)
    (bind _returnvalue_2 (GameplayTags|HasAnyMatchingGameplayTags(Message) (Variables|GetTargetActor) (Variables|GetTags_AlwaysVisible)))
    (if (not _returnvalue_2)
      (if (and (Variables|GetCanAlwaysSeeIndicatedActorsForOwnerController) (Utilities|IsValid (Variables|GetOwnerController) :!class /Script/Engine.KismetSystemLibrary))
        (if (IndicatedManagement|IsActorinIndicatedListforController (Variables|GetOwnerController) (Variables|GetTargetActor))
          (return true)
          (else
            _))
        (else
          _))
      (if (<= (Variables|GetMaxRange) 0)
        (LocalVariables|SetTraceEndLocation (Variables|GetAim_TargetLocation))
        (else
          (LocalVariables|SetTraceEndLocation (+ (Variables|GetAim_SourceLocation) (* (Math|Vector|Normalize (- (Variables|GetAim_TargetLocation) (Variables|GetAim_SourceLocation))) (Variables|GetMaxRange))))))
      (bind _asfort_player_pawn (Utilities|Casting|CastToFortPlayerPawn (Variables|GetTargetActor))
        (:then
          _)
        (:CastFailed
          _))
      (bind _vehicle_occupied (Utilities|IsinVehicle _asfort_player_pawn)
        (:true
          (bind _returnvalue_3 (Vehicle|GetVehicleActor _vehicle_occupied :!class /Script/FortniteGame.FortVehicleInterface))
          (Utilities|Array|Add (Variables|GetActorsToIgnore) _returnvalue_3))
        (:false
          _))
      (bind (_outhit _returnvalue_4) (Collision|LineTraceByChannel (Variables|GetTargetActor) (Variables|GetAim_SourceLocation) (LocalVariables|GetTraceEndLocation) (Variables|GetTraceChannel) false (Variables|GetActorsToIgnore) "None" false "(R=1.000000,G=0.000000,B=0.000000,A=1.000000)" "(R=0.000000,G=1.000000,B=0.000000,A=1.000000)" 2.0))
      (if _returnvalue_4
        (LocalVariables|SetHitActor (.HitActor (Collision|BreakHitResult _outhit)))
        (if (== (LocalVariables|GetHitActor) (Variables|GetTargetActor))
          (return true)
          (else
            (bind _returnvalue_5 (GameplayTags|HasMatchingGameplayTag(Message) (LocalVariables|GetHitActor) "(TagName=\"Gameplay.Melee.Block\")"))
            (if _returnvalue_5
              (return (and (or (== (Actor|GetAttachParentActor (LocalVariables|GetHitActor)) (Variables|GetTargetActor)) (== (Actor|GetOwner (LocalVariables|GetHitActor)) (Variables|GetTargetActor))) (Variables|GetCanSeeThroughMeleeBlock)))
              (else
                (bind _asfort_player_pawn_2 (Utilities|Casting|CastToFortPlayerPawn (Variables|GetTargetActor))
                  (:then
                    (return (and (or (== (Class|FortPlayerPawn|GetDBNOHoistee _asfort_player_pawn_2) (LocalVariables|GetHitActor)) (== (Class|FortPlayerPawn|GetDBNOHoister _asfort_player_pawn_2) (LocalVariables|GetHitActor))) (Variables|GetCanTargetDBNOCarried))))
                  (:CastFailed
                    (return false)))))))
        (else
          (return false)))
      (else
        (return true)))
    (else
      (return false))))
